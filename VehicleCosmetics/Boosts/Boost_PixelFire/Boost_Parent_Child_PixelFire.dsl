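; Vehicle boost cosmetic Blueprint for the Pixel Fire boost, driving a Niagara boost effect.
; When the cosmetic is applied, it assigns its NiagaraSystem asset to the NS_Boost component and calls SetupBoostCount.
; It then adjusts the boost offset per car through a shared cosmetics function library, using the vehicle cosmetics interface.
; BeginPlay, Tick and overlap events are empty, and the construction script only calls its parent.
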(event EventBeginPlay)

(event Collision|EventActorBeginOverlap
  (params (OtherActor :type Actor%20Object%20Reference)))

(event EventTick
  (params (DeltaSeconds :type Float%20%28single-precision%29)))

(event Vehicle|Cosmetics|EventOnCosmeticApplied
  (bind _ns_boost (Variables|Default|GetNS_Boost))
  (Niagara|SetNiagaraSystemAsset _ns_boost (Variables|Default|GetNiagaraSystem) true)
  (CallFunction|SetupBoostCount)
  (Class|BPVehicleCosmeticsFunctions|SetBoostOffsetbyCar (Variables|Vehicle|Cosmetics|GetVehicleCosmeticsInterface) _ns_boost))

(fn ConstructionScript
  (|CallParentFunction))
