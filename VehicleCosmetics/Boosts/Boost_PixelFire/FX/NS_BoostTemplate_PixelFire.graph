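# Niagara system for the Pixel Fire vehicle boost cosmetic, with one sprite emitter spawning flame particles from the car's skeletal mesh sockets.
# Particles split across boost ports, snap to right-angle rotations, shoot backward, grow, and play curve-driven SubUV animation.
# Painted-variant gradients color them over life, with RGB boosted for glow, while inherited vehicle movement fades out.
# Game mode drives spawn rate, lifetime and drag; frontend preview lengthens lifetime.
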NiagaraSystem /VehicleCosmetics/Boosts/Boost_PixelFire/FX/NS_BoostTemplate_PixelFire
dependencies {
 usedDataInterfaces [
  &NiagaraDataInterfaceArrayFloat
  &NiagaraDataInterfaceCurve
  &NiagaraDataInterfaceColorCurve
  &NiagaraDataInterfaceSpriteRendererInfo
  &NiagaraDataInterfaceSkeletalMesh
  &NiagaraDataInterfaceArrayFloat3
 ]
 usedDynamicInputs [&SelectVectorFromArray]
 usedModules [
  &self:SystemSpawnScript.SystemScriptSource.SystemScriptGraph.NiagaraNodeAssignment_0.SetVariables_#1
  &self:SystemSpawnScript.SystemScriptSource.SystemScriptGraph.NiagaraNodeAssignment_1.SetVariables_#2
  &SystemState
  &self:SystemSpawnScript.SystemScriptSource.SystemScriptGraph.NiagaraNodeAssignment_2.SetVariables_#3
 ]
}
emitters {NE_VC_Boost_Sprite_Parent002 {
  emitterUpdateScript [
   &EmitterState()
   &VC_FilteredSocketCount()
   &SpawnRate(
    SpawnRate=&MergeHelper_FloatSwitch(0=90, 1=90, 2=90, 3=270, Index=$User.GameModeID, Num=::NewEnumerator2)
    Use Spawn Probability=false
   )
   &SpawnRate(
    SpawnGroup=1
    SpawnRate=&MergeHelper_FloatSwitch(0=70, 1=70, 2=70, 3=210, Index=$User.GameModeID, Num=::NewEnumerator2)
    Use Spawn Probability=false
   ) {enabled=false, module=SpawnRate002}
  ]
  particleSpawnScript [
   /SetVariables(Particles.vistag:NiagaraInt32=&SpawnGroup()) {isSetParameters=true}
   &VC_PaintedVariantsOverride()
   &InitializeParticle(
    Color=$Particles.Color
    Lifetime=&MaskFloatBySpawnGroup(
     Masked Float=&MakeCustomFloatFromBool(
      Bool=$User.bIsFrontendPreview
      True Float=&RandomRangeFloat(
       Minimum=&MergeHelper_FloatSwitch(0=0.7, 1=0.7, 2=0.7, 3=0.7, Index=$User.GameModeID, Num=::NewEnumerator2)
       Maximum=&MergeHelper_FloatSwitch(0=1.5, 1=1.55, 2=1.5, 3=1.5, Index=$User.GameModeID, Num=::NewEnumerator2)
      )
      False Float=&RandomRangeFloat(
       Minimum=&MergeHelper_FloatSwitch(0=0.5, 1=0.5, 2=0.5, 3=0.5, Index=$User.GameModeID, Num=::NewEnumerator2)
       Maximum=&MergeHelper_FloatSwitch(0=1, 1=1, 2=1, 3=1, Index=$User.GameModeID, Num=::NewEnumerator2)
      )
     )
     Passthrough Float=0.3
    )
    Sprite Rotation Angle=&SelectFloatFromArray(
     Array Sampling Mode=::NewEnumerator0
     Direct Array Index=?
     Float Selection Array=DataInterface() {propertyValues {FloatData [0,90,180,270]}}
     Random Seed=0
    )
    Sprite Rotation Mode=::NewEnumerator0
    Sprite Size Mode=::NewEnumerator3
    Uniform Sprite Size=&Multiply_Float(A=70, B=2)
    UsePositionOffset=false
    Write Material Random=true
   )
   &PartitionParticles(Number Of Partitions=$Emitter.PortCount)
   &SkeletalMeshLocation(
    Bone / Socket Index=$Particles.PartitionParticles.Partition
    Bone / Socket Sampling Mode=::NewEnumerator5
    MaskSpawnGroup=false
    Mesh Sampling Type=::NewEnumerator6
    Sampled Position Offset=&AddVectorToPosition(Position=<0 0 0>, Vector=$User.DTCarOffset)
    Skeletal Mesh=$System.SkeletalMesh
   )
   &AddVelocity(Velocity=<-1000 0 100>)
   /SetVariables() {isSetParameters=true}
   /VehicleCosmetics/FX/VC_FXParents/Modules/DLR_ProductPreviews(GameModeID:NiagaraInt32=$System.GameModeID, bPreviewVehicleSpeed:NiagaraBool=$System.bPreviewVehicleSpeed)
  ]
  particleUpdateScript [
   &ParticleState()
   &ScaleSpriteSize(
    Scale Sprite Size Mode=::NewEnumerator0
    Uniform Scale Factor=&MaskFloatBySpawnGroup(
     Masked Float=&MakeFloatFromVector2d(Vector 2D=&Multiply_Vector2D_ByFloat(
       Float=&FloatFromCurve(FloatCurve=DataInterface() {propertyValues {Curve {keys [{interpMode=RCIM_Cubic, leaveTangent=2.6186, leaveTangentWeight=0.8873, tangentMode=RCTM_Break, time=0, value=0.6}, <at 1 2.3>]}}})
       Vector2D=<1 1>
      ))
     Passthrough Float=&MakeFloatFromVector2d(Vector 2D=&Multiply_Vector2D_ByFloat(
       Float=&FloatFromCurve(FloatCurve=DataInterface() {propertyValues {Curve {keys [<at 0 0.6665>,<at 1 0.7>]}}})
       Vector2D=<1 1>
      ))
    )
   )
   &VC_PaintedVariantsOverride(
    Black=&ColorFromCurve(ColorCurve=DataInterface() {propertyValues {
       AlphaCurve {keys [<at 0 30 RCIM_Cubic>,<at 1 30 RCIM_Cubic>]}
       BlueCurve {keys [<at 0 0.6771 RCIM_Cubic>, {arriveTangent=-0.6824, interpMode=RCIM_Cubic, leaveTangent=-0.6824, time=0.5423, value=0.0573}, <at 0.9922 0>]}
       GreenCurve {keys [<at 0 0.6771 RCIM_Cubic>, {arriveTangent=-0.6824, interpMode=RCIM_Cubic, leaveTangent=-0.6824, time=0.5423, value=0.0573}, <at 0.9922 0>]}
       RedCurve {keys [<at 0 0.6771 RCIM_Cubic>, {arriveTangent=-0.6824, interpMode=RCIM_Cubic, leaveTangent=-0.6824, time=0.5423, value=0.0573}, <at 0.9922 0>]}
      }})
    BurntSienna=&ColorFromCurve(ColorCurve=DataInterface() {propertyValues {
       AlphaCurve {keys [<at 0 30 RCIM_Cubic>,<at 1 30 RCIM_Cubic>]}
       BlueCurve {keys [
         <at 0 0>
         {arriveTangent=0.004, interpMode=RCIM_Cubic, leaveTangent=0.004, time=0.4222, value=0.2236}
         {arriveTangent=-0.3923, interpMode=RCIM_Cubic, leaveTangent=-0.3923, time=0.7558, value=0.003}
         <at 0.9922 0>
        ]}
       GreenCurve {keys [
         <at 0 0.7643>
         {arriveTangent=-0.9758, interpMode=RCIM_Cubic, leaveTangent=-0.9758, time=0.4222, value=0.4966}
         {arriveTangent=-0.8714, interpMode=RCIM_Cubic, leaveTangent=-0.8714, time=0.7558, value=0.0267}
         <at 0.9922 0>
        ]}
       RedCurve {keys [
         <at 0 3.7391>
         {arriveTangent=-4.7887, interpMode=RCIM_Cubic, leaveTangent=-4.7887, time=0.4222, value=1.8438}
         {arriveTangent=-3.2349, interpMode=RCIM_Cubic, leaveTangent=-3.2349, time=0.7558, value=0.1198}
         <at 0.9922 0>
        ]}
      }})
    Cobalt=&ColorFromCurve(ColorCurve=DataInterface() {propertyValues {
       AlphaCurve {keys [<at 0 30 RCIM_Cubic>,<at 1 30 RCIM_Cubic>]}
       BlueCurve {keys [<at 0 5 RCIM_Cubic>, {arriveTangent=-5.0395, interpMode=RCIM_Cubic, leaveTangent=-5.0395, time=0.6738, value=1}, <at 0.9922 0>]}
       GreenCurve {keys [<at 0 0.6013 RCIM_Cubic>, {arriveTangent=-0.606, interpMode=RCIM_Cubic, leaveTangent=-0.606, time=0.6738, value=0.0786}, <at 0.9922 0>]}
       RedCurve {keys [<at 0 0 RCIM_Cubic>,<at 0.6738 0 RCIM_Cubic>,<at 0.9922 0>]}
      }})
    Crimson=&ColorFromCurve(ColorCurve=DataInterface() {propertyValues {
       AlphaCurve {keys [<at 0 30 RCIM_Cubic>,<at 1 30 RCIM_Cubic>]}
       BlueCurve {keys [<at 0 0 RCIM_Cubic>,<at 0.6738 0 RCIM_Cubic>,<at 0.9922 0>]}
       GreenCurve {keys [<at 0 0.051 RCIM_Cubic>, {arriveTangent=-0.0514, interpMode=RCIM_Cubic, leaveTangent=-0.0514, time=0.6738, value=0.013}, <at 0.9922 0>]}
       RedCurve {keys [<at 0 1.9271 RCIM_Cubic>, {arriveTangent=-1.9423, interpMode=RCIM_Cubic, leaveTangent=-1.9423, time=0.6738, value=1}, <at 0.9922 0>]}
      }})
    DefaultColor=&ColorFromCurve(ColorCurve=DataInterface() {propertyValues {
       AlphaCurve {keys [<at 0 1>,<at 1 30 RCIM_Cubic>]}
       BlueCurve {keys [<at 0 0.9>, <at 0.2735 0.4>, {arriveTangent=-0.5506, interpMode=RCIM_Cubic, leaveTangent=-0.5506, time=0.6929, value=0}, <at 1 0>]}
       GreenCurve {keys [<at 0 5.7003>, <at 0.2735 1.9836>, {arriveTangent=-2.7305, interpMode=RCIM_Cubic, leaveTangent=-2.7305, time=0.6929, value=0.1203}, <at 1 0>]}
       RedCurve {keys [<at 0 30>, <at 0.2735 10>, {arriveTangent=-13.7653, interpMode=RCIM_Cubic, leaveTangent=-13.7653, time=0.6929, value=1}, <at 1 0>]}
      }})
    ForestGreen=&ColorFromCurve(ColorCurve=DataInterface() {propertyValues {
       AlphaCurve {keys [<at 0 30 RCIM_Cubic>,<at 1 30 RCIM_Cubic>]}
       BlueCurve {keys [<at 0 0 RCIM_Cubic>,<at 0.6738 0 RCIM_Cubic>,<at 0.9922 0>]}
       GreenCurve {keys [<at 0 5 RCIM_Cubic>, {arriveTangent=-5.0395, interpMode=RCIM_Cubic, leaveTangent=-5.0395, time=0.6738, value=0.4583}, <at 0.9922 0>]}
       RedCurve {keys [<at 0 0.2627 RCIM_Cubic>, {arriveTangent=-0.2648, interpMode=RCIM_Cubic, leaveTangent=-0.2648, time=0.6738, value=0.0292}, <at 0.9922 0>]}
      }})
    Grey=&ColorFromCurve(ColorCurve=DataInterface() {propertyValues {
       AlphaCurve {keys [<at 0 30 RCIM_Cubic>,<at 1 30 RCIM_Cubic>]}
       BlueCurve {keys [<at 0 0.6771 RCIM_Cubic>, {arriveTangent=-0.6824, interpMode=RCIM_Cubic, leaveTangent=-0.6824, time=0.5423, value=0.8177}, <at 0.9922 0>]}
       GreenCurve {keys [<at 0 0.6771 RCIM_Cubic>, {arriveTangent=-0.6824, interpMode=RCIM_Cubic, leaveTangent=-0.6824, time=0.5423, value=0.8177}, <at 0.9922 0>]}
       RedCurve {keys [<at 0 0.6771 RCIM_Cubic>, {arriveTangent=-0.6824, interpMode=RCIM_Cubic, leaveTangent=-0.6824, time=0.5423, value=0.8177}, <at 0.9922 0>]}
      }})
    Lime=&ColorFromCurve(ColorCurve=DataInterface() {propertyValues {
       AlphaCurve {keys [<at 0 30 RCIM_Cubic>,<at 1 30 RCIM_Cubic>]}
       BlueCurve {keys [<at 0 0 RCIM_Cubic>,<at 0.6738 0 RCIM_Cubic>,<at 0.9922 0>]}
       GreenCurve {keys [<at 0 5 RCIM_Cubic>, {arriveTangent=-4.7822, interpMode=RCIM_Cubic, leaveTangent=-4.7822, time=0.6738, value=0.7135}, <at 0.9922 0.2552>]}
       RedCurve {keys [
         <at 0 3.6062 RCIM_Cubic>
         {arriveTangent=-3.5824, interpMode=RCIM_Cubic, leaveTangent=-3.5824, time=0.6738, value=0.1614}
         <at 0.9922 0.0519>
        ]}
      }})
    Orange=&ColorFromCurve(ColorCurve=DataInterface() {propertyValues {
       AlphaCurve {keys [<at 0 30 RCIM_Cubic>,<at 1 30 RCIM_Cubic>]}
       BlueCurve {keys [<at 0 0 RCIM_Cubic>,<at 0.5423 0 RCIM_Cubic>,<at 0.8892 0>]}
       GreenCurve {keys [<at 0 0.8041 RCIM_Cubic>, {arriveTangent=-0.9043, interpMode=RCIM_Cubic, leaveTangent=-0.9043, time=0.5423, value=0.6452}, <at 0.8892 0>]}
       RedCurve {keys [<at 0 10 RCIM_Cubic>, {arriveTangent=-11.2456, interpMode=RCIM_Cubic, leaveTangent=-11.2456, time=0.5423, value=5}, <at 0.8892 0>]}
      }})
    Pink=&ColorFromCurve(ColorCurve=DataInterface() {propertyValues {
       AlphaCurve {keys [<at 0 30 RCIM_Cubic>,<at 1 30 RCIM_Cubic>]}
       BlueCurve {keys [<at 0 1.25 RCIM_Cubic>, {interpMode=RCIM_Cubic, leaveTangent=-0.2811, time=0.5423, value=4.3109}, <at 1 1>]}
       GreenCurve {keys [<at 0 0.4608 RCIM_Cubic>, {interpMode=RCIM_Cubic, leaveTangent=-0.5182, time=0.5423, value=0.599}, <at 1 0>]}
       RedCurve {keys [<at 0 1.2422 RCIM_Cubic>, {interpMode=RCIM_Cubic, leaveTangent=-1.1897, time=0.5423, value=5}, <at 1 0.1843>]}
      }})
    Purple=&ColorFromCurve(ColorCurve=DataInterface() {propertyValues {
       AlphaCurve {keys [<at 0 30 RCIM_Cubic>,<at 1 30 RCIM_Cubic>]}
       BlueCurve {keys [<at 0 5 RCIM_Cubic>, {arriveTangent=-5.6228, interpMode=RCIM_Cubic, leaveTangent=-5.6228, time=0.5423, value=3}, <at 0.8892 0>]}
       GreenCurve {keys [<at 0 0.2344 RCIM_Cubic>, {arriveTangent=-0.2636, interpMode=RCIM_Cubic, leaveTangent=-0.2636, time=0.5423, value=0}, <at 0.8892 0>]}
       RedCurve {keys [<at 0 1.3298 RCIM_Cubic>, {arriveTangent=-1.4954, interpMode=RCIM_Cubic, leaveTangent=-1.4954, time=0.5423, value=1.3604}, <at 0.8892 0>]}
      }})
    Saffron=&ColorFromCurve(ColorCurve=DataInterface() {propertyValues {
       AlphaCurve {keys [<at 0 30 RCIM_Cubic>,<at 1 30 RCIM_Cubic>]}
       BlueCurve {keys [<at 0 0 RCIM_Cubic>,<at 0.5423 0 RCIM_Cubic>,<at 0.9864 0>]}
       GreenCurve {keys [
         <at 0 2.4536 RCIM_Cubic>
         {arriveTangent=-2.2215, interpMode=RCIM_Cubic, leaveTangent=-2.2215, time=0.5423, value=2.2201}
         <at 0.9864 0.2622>
        ]}
       RedCurve {keys [<at 0 5 RCIM_Cubic>, {arriveTangent=-4.7466, interpMode=RCIM_Cubic, leaveTangent=-4.7466, time=0.5423, value=3}, <at 0.9864 0.3177>]}
      }})
    SkyBlue=&ColorFromCurve(ColorCurve=DataInterface() {propertyValues {
       AlphaCurve {keys [<at 0 30 RCIM_Cubic>,<at 1 30 RCIM_Cubic>]}
       BlueCurve {keys [<at 0 5 RCIM_Cubic>, {arriveTangent=-5.6228, interpMode=RCIM_Cubic, leaveTangent=-5.6228, time=0.5423, value=3}, <at 0.8892 0>]}
       GreenCurve {keys [<at 0 4.8914 RCIM_Cubic>, {arriveTangent=-5.5007, interpMode=RCIM_Cubic, leaveTangent=-5.5007, time=0.5423, value=1.2156}, <at 0.8892 0>]}
       RedCurve {keys [<at 0 0 RCIM_Cubic>,<at 0.5423 0 RCIM_Cubic>,<at 0.8892 0>]}
      }})
    TitaniumWhite=<rgba 3 3 3 20>
   ) {module=VC_PaintedVariantsOverride001}
   &InheritSourceMovement(
    Applied Position Scale=&Multiply_VectorByFloat(
     Vector=<1 1 1>
     Float=&FloatFromCurve(FloatCurve=DataInterface() {propertyValues {Curve {keys [
          {interpMode=RCIM_Cubic, leaveTangent=-0.0051, tangentMode=RCTM_Break, time=0, value=1}
          {arriveTangent=-0.247, interpMode=RCIM_Cubic, tangentMode=RCTM_Break, time=1, value=0.9}
         ]}}})
    )
    ApplyForce=false
    ApplyPosition=true
    ApplyPositionFalloff=false
   ) {enabled=false}
   &Drag(
    Drag=&MergeHelper_FloatSwitch(0=1, 1=1, 2=1, 3=0, Index=$User.GameModeID, Num=::NewEnumerator2)
    Use Linear Drag=true
    Use Rotational Drag=true
   )
   &InheritSourceMovement(
    Applied Position Scale=&Multiply_VectorByFloat(
     Vector=<1 1 1>
     Float=&FloatFromCurve(FloatCurve=DataInterface() {propertyValues {Curve {
         defaultValue=1
         keys [
          {arriveTangent=-1.7885, interpMode=RCIM_Cubic, leaveTangent=-1.7885, tangentMode=RCTM_User, time=0.05, value=1}
          {interpMode=RCIM_Cubic, tangentMode=RCTM_User, time=1, value=0}
         ]
        }}})
    )
    ApplyForce=false
    ApplyPosition=true
    ApplyPositionFalloff=false
   ) {module=InheritSourceMovement001}
   &SubUVAnimation(
    "Sprite Renderer "=DataInterface() {propertyValues {SpriteRenderer=&self:NE_VC_Boost_Sprite_Parent002_0.NiagaraSpriteRendererProperties_7}}
    SubUV Animation Mode:ENiagara_SubUVLookupModeV2=::NewEnumerator1
    SubUV Curve=DataInterface() {propertyValues {Curve {
       defaultValue=0
       keys [
        {
         interpMode=RCIM_Cubic
         leaveTangent=2.9578
         leaveTangentWeight=0.446
         tangentMode=RCTM_Break
         tangentWeightMode=RCTWM_WeightedBoth
         time=0.025
         value=0.02
        }
        {arriveTangent=0.5673, interpMode=RCIM_Cubic, leaveTangent=0.5673, tangentMode=RCTM_User, time=0.4295, value=0.45}
        {
         arriveTangent=2.5482
         arriveTangentWeight=0.3328
         interpMode=RCIM_Cubic
         leaveTangent=1.4188
         tangentMode=RCTM_Break
         tangentWeightMode=RCTWM_WeightedBoth
         time=1
         value=1
        }
       ]
      }}}
    UseEndFrame=false
    UseStartFrame=true
   )
   /SetVariables(Particles.1-SpriteColor:LinearColor=/DealershipSharedLibrary/FX/Niagara/DynamicInputs/MultiplyLinearColorByVectorAndFloat(
     Color to Scale:LinearColor=$Particles.Color
     RGB Scale:Vector3f=&VectorFromFloat(Value=5)
     Alpha Scale:NiagaraFloat=1
    )) {isSetParameters=true}
   &SolveForcesAndVelocity(Clamp Velocity=false, Limit Acceleration=false)
  ]
  rendererClasses [&NiagaraSpriteRendererProperties]
  renderers [
   {rendererClass=&NiagaraSpriteRendererProperties}
   {rendererClass=&NiagaraSpriteRendererProperties, rendererIndex=1}
  ]
 }}
summary {
 systemName=NS_BoostTemplate_PixelFire
 emitters [{emitterName=NE_VC_Boost_Sprite_Parent002, rendererClasses [&NiagaraSpriteRendererProperties]}]
 userVariables [
  User.bIsFrontendPreview {type=NiagaraBool}
  User.bIsPainted {type=NiagaraBool}
  User.blsIsCoroColor {type=NiagaraBool}
  User.bPreviewVehicleSpeed {type=NiagaraBool}
  User.ColorVariantID {type=NiagaraInt32}
  User.DTCarOffset {type=Vector3f, description="", defaultValue {struct=Vector3f, value=<0 0 0>}}
  User.GameModeID {type=NiagaraInt32, defaultValue=3}
  User.PaintedColor {type=LinearColor, description="", defaultValue {struct=LinearColor, value=<rgba 1 1 1 1>}}
  User.PaintedScalar {type=NiagaraFloat, defaultValue=1, description=""}
  User.PortCount {type=NiagaraInt32, defaultValue=2}
 ]
}
